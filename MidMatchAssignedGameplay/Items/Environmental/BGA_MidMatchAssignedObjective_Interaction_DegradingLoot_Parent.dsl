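; Parent Blueprint for a mid-match assigned objective cache whose reward loot rarity degrades while it waits to be activated.
; Server timers drive an anticipation flag and stepped rarity drops, replicated via OnRep for cosmetics, with hotfixed non-degrading fallbacks and Creative playlist-tag overrides.
; Activation locks rarity, clears timers and records a squad stat.
; Rewards map rarity and squad size to loot tiers, then spawn and radially toss items with ammo.
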(event Custom|OnDecayLootRarityChanged_Cosmetics)

(event Custom|BeginDegradeAnticipation_Cosmetics)

(event EventObjectiveState_AvailableIdle
  ;[0.2824 0.3886 0.4149 1 120] Objective State: AvailableIdle - start degrading
  ;;[0.03332 0.3212 0.2859] OnRiftSpawnCompleted
  ;; Sets the degrading timer to when the objective is effectively spawned; also Calls parent that finishes the spawning of the objective and cosmetics
  (|CallParentFunction)
  (Networking|SwitchHasAuthority
    (:Authority
      (bind _continue (CallFunction|SearchForCreativeRoyaleTags))
      (if _continue
        (if (ScalableFloat|GetValueasBool (Variables|Objective|Degrading|GetHF_IsDegradingLootEnabled))
          (CallFunction|BeginDegradingTimers)
          (else
            ;;[0.6885 0 1] Fallback Loot Rules - Loot will not degrade. set by hotfix. The two other rulesets are:
            ;; 1) use RNG of the valid loot tables OR 2) a static rarity for everyone - all loot is still dependent on the playlist
            ;;;[0.4158 0.6112 1] When using a fallback Non-Degrading loot style, just set the rarity once. Will also apply the appropriate visuals via OnRep on the client
            (Variables|Objective|Degrading|SetCurrentDegradedLootRarity (select (ScalableFloat|GetValueasBool (Variables|Objective|Degrading|Fallback|GetHF_NonDegradingShouldBeRNG)) (Utilities|Enum|BytetoEnumEFortRarity (Math|Conversions|ToByte(Integer) (Math|Random|RandomIntegerinRange (ScalableFloat|GetValueasInteger (Variables|Objective|Degrading|Fallback|GetHF_NonDegradingRNGLootMin)) (ScalableFloat|GetValueasInteger (Variables|Objective|Degrading|Fallback|GetHF_NonDegragingRNGLootMax))))) (Utilities|Enum|BytetoEnumEFortRarity (Math|Conversions|ToByte(Integer) (ScalableFloat|GetValueasInteger (Variables|Objective|Degrading|Fallback|GetHF_NonDegradingStaticLootLevel))))))))))))

(event Custom|BeginDegradingTimers
  (bind _timeslootdegradedcount (Variables|Objective|Degrading|GetTimesLootDegradedCount))
  (bind _returnvalue (Math|Conversions|ToFloat(Integer) _timeslootdegradedcount))
  (bind _hf_timetodegradewithlevels (Variables|Objective|Degrading|GetHF_TimeToDegradeWithLevels))
  (bind _returnvalue_2 (ScalableFloat|GetValueatLevel _hf_timetodegradewithlevels _returnvalue))
  ;[0.2824 0.3886 0.4149 1 120] Objective State: AvailableIdle - start degrading
  ;;[0.03794 1 0.09435] BeginDegradingTimers
  ;; Set timers for the degrading and anticipation
  (bind _returnvalue_3 (Utilities|Time|SetTimerbyEvent (EventDispatchers|CreateEvent :!event Custom|OnDegradeRewardLootRarity) _returnvalue_2))
  (Variables|Objective|Degrading|SetTimer_DegradingLoot _returnvalue_3)
  (bind _returnvalue_4 (Utilities|Time|SetTimerbyEvent (EventDispatchers|CreateEvent :!event Custom|BeginDegradeAnticipation) (- _returnvalue_2 (ScalableFloat|GetValueatLevel (Variables|Objective|Degrading|GetHF_DegradeWarningTime) (Math|Conversions|ToFloat(Integer) (Variables|Objective|Degrading|GetTimesLootDegradedCount))))))
  (Variables|Objective|Degrading|SetTimer_DegradingAnticipation _returnvalue_4)
  (Variables|Objective|Degrading|SetbShouldBeginDegradeAnticipation false))

(event Custom|BeginDegradeAnticipation
  ;[0.2824 0.3886 0.4149 1 120] Objective State: AvailableIdle - start degrading
  ;;[0.4219 0.2276 0.07989] BeginDegradeAnticipation
  ;; Replicate to the client for VFX
  (Variables|Objective|Degrading|SetbShouldBeginDegradeAnticipation true))

(event Custom|OnDegradeRewardLootRarity
  ;[0.2824 0.3886 0.4149 1 120] Objective State: AvailableIdle - start degrading
  ;;[0.4219 0.2276 0.07989] OnDegradeRewardLootRarity
  ;; Degrade the rarity and determine if an additional degrading timer is needed to start to degrade further
  ;;;[0.4219 0.2276 0.07989] Degrade the rarity
  (bind _result (Math|Integer|IncrementInt (Variables|Objective|Degrading|GetTimesLootDegradedCount))
    (:%20%20
      (Variables|Objective|Degrading|SetCurrentDegradedLootRarity (Utilities|Enum|BytetoEnumEFortRarity (- (Variables|Objective|Degrading|GetCurrentDegradedLootRarity) 1))) ; replicate to clients for visual updates
      ;;;[0.4158 0.6112 1] Check if the loot rarity should continue to degrade and set a new timer
      (if (< _result (ScalableFloat|GetValueatLevel (Variables|Objective|Degrading|GetHF_MaxTimesToDegrade)))
        (CallFunction|BeginDegradingTimers)
        (else
          (Variables|Objective|Degrading|SetbShouldBeginDegradeAnticipation false) ; make sure to disable so anticipation \neffects don't play again based on relevancy
          (Networking|SetNetDormancy "DORM_DormantAll") ; If we're done degrading, set ourselves dormant
        )))))

(event EventCosmeticsState_ActiveObjective_PersistentLoop
  ;[0.04953 0 0.1354 1 120] CLIENT COSMETIC EVENTS (VFX/SFX)
  ;;[0.04953 0 0.1354 1 50] COSMETICS: ActiveObjective
  ;;;[0.3658 0 1] CosmeticsState_ActiveObjective_PersistentLoop
  ;;; Lock in the current rarity and stop degrading communication
  (|CallParentFunction)
  ;;;;[0.3658 0 1] Once the objective is locked in, the degradation visuals should stop...
  (CallFunction|EndDegradeAnticipationCosmetics)
  ;;;;[0.3658 0 1] ...and then lock in the current rarity
  (CallFunction|OnDecayLootRarityChangedCosmetics))

(event Custom|EndDegradeAnticipation_Cosmetics)

(event EventObjectiveState_ActiveObjective
  ;[0.2824 0.3886 0.4149 1 120] Objective State: ActiveObjective - lock in the current degraded loot level
  ;;[0.03332 0.3212 0.2859] ObjectiveState_ActiveObjective
  ;; Lock in the current loot rarity by clearing the degrading timer; Call parent to start the objective timer
  (|CallParentFunction)
  (Networking|SwitchHasAuthority
    (:Authority
      ;;;[0 0.04199 1] Stop Degrading the loot
      (Utilities|Time|ClearandInvalidateTimerbyHandle (Variables|Objective|Degrading|GetTimer_DegradingLoot))
      (Utilities|Time|ClearandInvalidateTimerbyHandle (Variables|Objective|Degrading|GetTimer_DegradingAnticipation))
      (Variables|Objective|Degrading|SetbShouldBeginDegradeAnticipation false)
      ;;[0.4189 1 0.7234] Set stat on squad who activated the cache (for tourney stats)
      (Default|SetCacheActivatedStat))))

(event EventSpawnRewardItems
  (bind _currentdegradedlootrarity (Variables|Objective|Degrading|GetCurrentDegradedLootRarity))
  (bind (_rewardloottier _foundtier) (CallFunction|GetCurrentLootTierbyRarity _currentdegradedlootrarity))
  ;;[0 0.04199 1] SpawnRewardItems
  ;; OVERRIDE
  ;; Determines the reward tier to give based on the playlist and current status of the degrade timer
  ;[0.2824 0.3886 0.4149 1 120] Objective State: CompletedIdle - grant the reward
  ;;;[0 0.04199 1] Spawn loot
  ;;; When using a fallback instead of Degrading Loot, "CurrentDegradedLootLevel" is already set to the static loot
  (if _foundtier
    (bind _rewarditempickups (CallFunction|SpawnObjectiveRewardLoot _rewardloottier))
    (CallFunction|TossSpawnedRewardLoot _rewarditempickups)
    (else
      ;;;;[1 0 0.9627] Fallback to Parent logic and spawn default loot if failed
      (Fortnite|FortShippingLog self "Assigned Cache failed to determine loot from playlist/degrade level" true)
      (|CallParentFunction) ; intentionally does not call parent\nexcept in fallback case
    )))

(event EventEndPlay
  (params (EndPlayReason :type EEndPlayReason%20Enum :default "Destroyed"))
  ;;[1 0.3501 0] Event EndPlay
  ;; Cleanup: clear timers, unbind events, etc.
  ;[0.2824 0.3886 0.4149 1 120] END PLAY - cleanup and destruction
  (|CallParentFunction EndPlayReason)
  ;;;[1 0.3501 0] Clear timers
  (Utilities|Time|ClearandInvalidateTimerbyHandle (Variables|Objective|Degrading|GetTimer_DegradingLoot))
  (Utilities|Time|ClearandInvalidateTimerbyHandle (Variables|Objective|Degrading|GetTimer_DegradingAnticipation)))

(fn BlueprintGetInteractionString
  (params (InteractingPawn :type Fort%20Pawn%20Object%20Reference) (InteractionBeingAttempted :type EInteractionBeingAttempted%20Enum :default "FirstInteraction"))
  (bind _currentdegradedlootrarity (Variables|Objective|Degrading|GetCurrentDegradedLootRarity))
  ;[0.5877 0 1] BlueprintGetInteractionString
  ; Sets the interaction string display text based on objective assignment and current degraded loot level
  ;[Extends Parent]
  (bind _returnvalue (|CallParentFunction InteractingPawn InteractionBeingAttempted))
  (if (not (Utilities|Text|TextIsEmpty _returnvalue))
    (bind _returnvalue_2 (Fortnite|GetRarityData _currentdegradedlootrarity))
    ;;[1 0.3594 0.9217] Return: append rarity text to the interaction
    (return (Utilities|Text|FormatText "INVTEXT(\"{0}\\r\\n{1}\")" _returnvalue (Utilities|Text|FormatText (Utilities|Select :Index _currentdegradedlootrarity :Common "INVTEXT(\"<Rarity_Common>({0})</>\")" :Uncommon "INVTEXT(\"<Rarity_Uncommon>({0})</>\")" :Rare "INVTEXT(\"<Rarity_Rare>({0})</>\")" :Epic "INVTEXT(\"<Rarity_Epic>({0})</>\")" :Legendary "INVTEXT(\"<Rarity_Legendary>({0})</>\")" :Mythic "INVTEXT(\"<Rarity_Mythic>({0})</>\")") (Utilities|Struct|BreakFortRarityItemData _returnvalue_2))))
    (else
      ;;[1 0.3594 0.9217] Return: null
      (return ""))))

(fn GetAnalyticsAttributes
  (bind _currentdegradedlootrarity (Variables|Objective|Degrading|GetCurrentDegradedLootRarity))
  (bind _returnvalue (Utilities|String|EnumtoString _currentdegradedlootrarity))
  ;[0.2435 0.1798 0.4062] GetAnalyticsAttributes [Extends Parent]
  ; Add an analytics call to send the degraded rarity when the assigned objective had been activated
  (bind _analyticseventattributes (|CallParentFunction))
  (LocalVariables|SetReturn_DegradedLootAnalytics _analyticseventattributes) ; set the local variable to the parent's  returned \nattributes to add any degraded specific attributes
  (Utilities|FlowControl|Sequence
    (:then_0
      ;;[0.2435 0.1798 0.4062] Filter rarity degradation unless the objective was activated
      ;; - it was irrelevant to the game and can be assumed to be Rare (Blue)
      (if (> (Math|Byte|MakeLiteralByte (Variables|Objective|States|GetECurrentObjectiveState_RepNotify)) 2)
        ;;[0.2435 0.1798 0.4062] Attribute: What was the rarity level locked at when the objective was activated?
        (bind _returnvalue_2 (Analytics|ConstructAnalyticsAttribute "DegradedLootRarityOnActivation" _returnvalue))
        (Utilities|Array|Add (LocalVariables|GetReturn_DegradedLootAnalytics) _returnvalue_2)
        (Fortnite|FortShippingLog self (Utilities|String|Append "DEGRADING OBJECTIVE ANALYTICS:  DegradedLootRarityOnActivation = " _returnvalue))))
    (:then_1
      ;;[0.9707 0.4876 1] RETURN: Send all the tracked analytic events (some are filtered based on the state of the objective.
      (return (LocalVariables|GetReturn_DegradedLootAnalytics)))))

(fn GetCurrentLootTierByRarity
  (params (Current%20Rarity :type EFortRarity%20Enum :default "Common"))
  (bind (_value _returnvalue) (Utilities|Map|Find (Utilities|Select :Index Current%20Rarity :Common (Variables|Objective|Reward|LootTiers|GetMap_RareLoot_PlaylistToLootTier) :Uncommon (Variables|Objective|Reward|LootTiers|GetMap_RareLoot_PlaylistToLootTier) :Rare (Variables|Objective|Reward|LootTiers|GetMap_RareLoot_PlaylistToLootTier) :Epic (Variables|Objective|Reward|LootTiers|GetMap_EpicLoot_PlaylistToLootTier) :Legendary (Variables|Objective|Reward|LootTiers|GetMap_LegendaryLoot_PlaylistToLootTier) :Mythic (Variables|Objective|Reward|LootTiers|GetMap_LegendaryLoot_PlaylistToLootTier) :Transcendent (Variables|Objective|Reward|LootTiers|GetMap_LegendaryLoot_PlaylistToLootTier) :Unattainable (Variables|Objective|Reward|LootTiers|GetMap_LegendaryLoot_PlaylistToLootTier)) (Reward|GetPlaylistSquadSizeTag)))
  ;[0 0.1068 0.6441] GetCurrentLootTierByRarity
  ; Maps the loot tier based on the playlist squad size and the curent rarity of the degraded loot
  (return _value _returnvalue))

(fn OnRep_bShouldBeginDegradeAnticipation
  ;[0.4455 0 1] ShouldBeingDegradeAnticipation
  ; Replicated to clients - starts the degrading anticipation; only happens in the AvailableIdle objective state when degrading is allowed
  (if (and (Variables|Objective|Degrading|GetbShouldBeginDegradeAnticipation) (Utilities|Enum|Equal(Enum) (Variables|Objective|States|GetECurrentObjectiveState_RepNotify) "NewEnumerator2"))
    (Networking|CanExecuteCosmeticEvents
      (:True
        (CallFunction|BeginDegradeAnticipationCosmetics)))))

(fn OnRep_CurrentDegradedLootRarity
  ;[0.6176 0 1] OnRep CurrentDegradedLootRarity
  ; Replicate the change to clients to update the rarity visuals on the objective
  (Networking|SwitchHasAuthority
    (:Remote
      (CallFunction|OnDecayLootRarityChangedCosmetics))))

(fn SearchforCreativeRoyaleTags
  (bind _creativeroyaletags (Variables|Objective|Degrading|GetCreativeRoyaleTags))
  (if (ScalableFloat|GetValueasBool (Variables|Creative|GetHotFix_UseDefaultPlaylistTagForCR))
    ;[0.6592 1 0.6458] Search for relevant tags set by the Loot Controller Device.
    (bind _array_element (Utilities|Array|ForEachLoopwithBreak (GameplayTags|BreakGameplayTagContainer (Class|FortGameStateAthena|GetAthenaPlaylistContextTags (Fortnite|Athena|GameState|GetGameStateAthena))))
      (:LoopBody
        (if (.ReturnValue (Utilities|Map|Find _creativeroyaletags _array_element))
          (Variables|Objective|Degrading|SetCurrentDegradedLootRarity (Utilities|Map|Find _creativeroyaletags _array_element))
          (LocalVariables|SetbContinueDegradationLogic false)))
      (:Completed
        _))
    (return (LocalVariables|GetbContinueDegradationLogic))
    (else
      (return true))))

(fn SetCacheActivatedStat
  (bind _returnvalue (Fortnite|Athena|GameState|GetGameStateAthena))
  (bind _returnvalue_2 (Actor|GetComponentbyClass _returnvalue "/MidMatchAssignedGameplay/Items/BP_AssignedObjectiveGameStateComponent.BP_AssignedObjectiveGameStateComponent_C"))
  ;[0.295 1 0.8834] Set the "objective completed" stat (BP_AssignedObjectiveGameStateComponent::ObjectiveCompletedStatName) for the player's squad who activted the cache. (For use with tournament stats)
  ; This isn't necessary "completing" the objective, but it is the stat we're tracking for this objective
  (Networking|SwitchHasAuthority
    (:Authority
      (Utilities|IsValid _returnvalue
        (:Is%20Valid
          (bind _asfort_player_state_athena (Utilities|Casting|CastToFortPlayerStateAthena (Variables|Objective|Interaction|GetPlayerStateThatStartedObjective)))
          (Utilities|IsValid _returnvalue_2
            (:Is%20Valid
              (Class|FortGameStateComponentMidMatchObjectiveParent|SetObjectiveCompletedStat _returnvalue_2 (Class|FortPlayerStateAthena|GetSquadId _asfort_player_state_athena)))))))))

(fn SpawnObjectiveRewardLoot
  (params (LootTier :type Name))
  ;[0 0.04968 1] SpawnObjectiveRewardLoot
  ; Get the loot to spawn from the Loot Tier set from degrading and any associated ammo drops
  (bind _itempickupsspawned (Class|AthenaFunctionLibrary|SpawnLootfromTableUsingItemEntry LootTier 0 -1 (Utilities|Operators|Add (Transformation|GetActorLocation) (Math|Vector|MakeVector 0.0 0.0 (ScalableFloat|GetValueatLevel (Variables|Objective|Reward|Spawning|GetHF_RewardLootHeightOffset)))) "1.000000,0.000000,0.000000" 0 true true false 0 "Tossed" "LootDrop"))
  (bind _output_get (LocalVariables|SetReturn_SpawnedLoot _itempickupsspawned))
  (for _array_element _output_get
    ;;[0 0.04968 1] For weapons, get the associated ammo and spawn that as well
    (bind _asfort_weapon_item_definition (Utilities|Casting|CastToFortWeaponItemDefinition (Class|FortItemFunctionLibrary|BreakFortItemEntry (Pickup|GetItemEntry _array_element 0))))
    (bind _returnvalue (Fortnite|Weapon|CreateTossAmmoPickupforWeaponItemDefinitionatLocation _asfort_weapon_item_definition (Variables|Objective|Reward|Spawning|GetRewardAmmoSourceTag) (Transformation|GetActorLocation)))
    (Utilities|IsValid _returnvalue
      (:Is%20Valid
        (Utilities|Array|Add (LocalVariables|GetReturn_SpawnedLoot) _returnvalue) ; add ammo to spawned items
      )))
  ;;[1 0.4103 0.9444] Return: All reward items and associated ammo
  (return (LocalVariables|GetReturn_SpawnedLoot)))

(fn TossSpawnedRewardLoot
  (params (SpawnedLootToToss :type Array%20of%20Fort%20Pickup%20Object%20References))
  (bind (_returnvalue_roll _returnvalue_pitch _returnvalue_yaw) (Math|Rotator|BreakRotator (Math|Rotator|MakeRotfromX (Transformation|GetActorForwardVector))))
  ;[0.006141 0 1] TossSpawnedRewardLoot
  ; Tosses the loot in a full circle radius and evenly distributes it over that area
  (for (_array_element _array_index) SpawnedLootToToss
    ;;[0.3166 0.4028 1] Use a more authored toss because the standard one is unreliable
    (Fortnite|Container|TossPickupfromContainer self _array_element (- (Utilities|Array|Length SpawnedLootToToss) 1) _array_index 3.6e+02 (Math|Rotator|MakeRotator 0.0 35.0 _returnvalue_yaw) (Math|Random|RandomFloatinRange (Math|Float|Max(Float) (ScalableFloat|GetValueatLevel (Variables|Objective|Reward|Spawning|GetHF_LootTossSpeedMin)) (Variables|Objective|Reward|Spawning|GetMinimumRequiredLootTossSpeed)) (Math|Float|Max(Float) (ScalableFloat|GetValueatLevel (Variables|Objective|Reward|Spawning|GetHF_LootTossSpeedMax)) (Variables|Objective|Reward|Spawning|GetMinimumRequiredLootTossSpeed))) false false 0)))

(fn ConstructionScript
  (|CallParentFunction))
