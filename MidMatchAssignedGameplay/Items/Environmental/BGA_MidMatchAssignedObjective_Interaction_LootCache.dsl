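; Loot cache interaction actor for mid-match assigned gameplay objectives, supplying the cache's open animation, hologram image and marker details ID.
; Opening is a vertex material animation: it writes the current time to custom primitive data and enables world position offset evaluation on the interaction mesh.
; The hologram image is set through a dynamic material instance's ProfileImage texture parameter.
; The construction script only calls the parent.
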(fn BeginChestOpenAnim
  ;[0.3363 0 1] BeginChestOpenAnim
  ; Sets variables for the vertex material animation to hopen the cache
  (Rendering|Material|SetCustomPrimitiveDataFloat (Variables|Default|GetInteractionObject) 0 (Utilities|Time|GetTimeSeconds)) ; pass in current time to sequence the animation
  (Rendering|LOD|SetEvaluateWorldPositionOffset (Variables|Default|GetInteractionObject) true :!class /Script/Engine.StaticMeshComponent) ; update the vertex animation
)

(fn GetMarkerDetailsId
  ;[0.2361 0.8293 1] GetMarkerDetailsID
  ; Passes id to get detailed marker data
  (return "MarkerDetails.MMObj.AssignedGameplay"))

(fn SetHologramImage
  (params (NewHologramImage :type Texture%20Object%20Reference))
  (bind _hologram (Variables|Default|GetHologram))
  ;[0.3518 0 1] SetHologramImage
  ; Updated the VFX hologram image with the input image
  (bind _returnvalue (Rendering|Material|CreateDynamicMaterialInstance _hologram 0 (Rendering|Material|GetMaterial _hologram 0) "None" :!class /Script/Engine.PrimitiveComponent))
  (bind _output_get (Variables|VFX|Hologram|SetHologramMat _returnvalue))
  (Rendering|Material|SetTextureParameterValue _output_get "ProfileImage" NewHologramImage)
  (Rendering|Material|SetMaterial _hologram 0 _returnvalue))

(fn ConstructionScript
  (|CallParentFunction))
