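# Surface material for the augment container item, animating its opening entirely in world position offset with Legendary diffuse, normal, mask and emissive textures as defaults.
# StartTime custom primitive data and TimeMult drive a curve atlas lookup, with a debug switch to scrub StartTime directly.
# Vertex color selects moving parts: red rotates about a pivot, green and blue slide along opposite local Z directions.
# Normals on the rotating part are corrected to follow the rotation.
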Material /MidMatchAssignedGameplay/Items/Environmental/Materials/M_AugmentContainer
graph {
 properties {bUseMaterialAttributes=true, pixelDepthOffsetMode=PDOM_Legacy, usedShadingModels=Default Lit}
 nodes [
  ComponentMask(None=Multiply(A=WorldPosition().XYZ, B=ScalarParameter() {defaultValue=512, parameterName=ssssss})) {g=true, r=true}
  Constant() {r=0.35}
  Constant()
  Frac(None=Multiply(A=Time() {period=0.0001}))
  n1 = TextureSampleParameter2D() {
   parameterName=SpecularMasks
   samplerType=SAMPLERTYPE_Masks
   texture=/MidMatchAssignedGameplay/Items/Environmental/Textures/T_Container_Augment_Legendary_S
  }
  n2 = Add(A=Add(A=NamedRerouteUsage(), B=NamedRerouteUsage()), B=NamedRerouteUsage())
  MakeMaterialAttributes!(
   BaseColor=TextureSampleParameter2D() {parameterName=Diffuse, texture=/MidMatchAssignedGameplay/Items/Environmental/Textures/T_Container_Augment_Legendary_D}.RGB
   Metallic=@n1.G
   Specular=@n1.G
   Roughness=@n1.G
   Anisotropy
   EmissiveColor=TextureSampleParameter2D() {parameterName=Emissive, texture=/MidMatchAssignedGameplay/Items/Environmental/Textures/T_Container_Augment_Legendary_E}.RGB
   Opacity
   OpacityMask
   Normal=NamedRerouteUsage()
   Tangent
   WorldPositionOffset=@n2
   SubsurfaceColor
   ClearCoat
   ClearCoatRoughness
   AmbientOcclusion
   Refraction
   CustomizedUVs_0
   CustomizedUVs_1
   CustomizedUVs_2
   CustomizedUVs_3
   CustomizedUVs_4
   CustomizedUVs_5
   CustomizedUVs_6
   CustomizedUVs_7
   PixelDepthOffset
   ShadingModel
   Displacement
  )
  MaterialFunctionCall() {materialFunction=&DebugScalarValues}
  MaterialFunctionCall!(
   Alpha=ScalarParameter() {bUseCustomPrimitiveData=true, defaultValue=-1, parameterName=StartTime}
   Color Input 0=Constant()
   Color Input 1=@n2
  ) {materialFunction=/Engine/Functions/Engine_MaterialFunctions02/Utility/DynamicBranch}
  NamedRerouteDeclaration(None=MaterialFunctionCall(Float3=NamedRerouteUsage()) {materialFunction=&BreakOutFloat3Components}.R) {name=Step1_AngleValue, nodeColor=<rgba 0 0.9062 1 1>}
  NamedRerouteDeclaration(None=Transform(None=Constant3Vector() {constant=<rgba 0 0 1 0>}) {transformSourceType=TRANSFORMSOURCE_Local}) {name=Step1_AxisValue, nodeColor=<rgba 0 1 0.6719 1>}
  NamedRerouteDeclaration(None=TransformPosition(None=VectorParameter() {parameterName=Step1_Pivot, defaultValue=<rgba 0 0 0 1>}.RGB) {transformType=TRANSFORMPOSSOURCE_World}) {name=Step1_PivotValue, nodeColor=<rgba 0.4531 1 0 1>}
  n3 = NamedRerouteUsage()
  n4 = NamedRerouteUsage()
  n5 = TextureSampleParameter2D() {
   parameterName=Normals
   samplerType=SAMPLERTYPE_Normal
   texture=/MidMatchAssignedGameplay/Items/Environmental/Textures/T_Container_Augment_Legendary_N
  }
  n6 = Transform(None=@n5.RGB)
  n7 = VectorParameter() {parameterName=Step1_Pivot, defaultValue=<rgba 0 0 0 1>}
  NamedRerouteDeclaration(None=If(
    A=VertexColor().R
    "A > B"=MaterialFunctionCall(Rotation Axis=@n3, Rotation Angle=@n4, World Space Vertex Normals=@n6, Normals Pivot=@n7.RGB) {materialFunction=&FixRotateAboutAxisNormals}.New Tangent Space Vertex Normal
    "A < B"=@n5.RGB
   ) {constB=0.9, equalsThreshold=0}) {name=FixedNormals, nodeColor=<rgba 1 0 0.7031 1>}
  n8 = NamedRerouteUsage()
  n9 = NamedRerouteUsage()
  n10 = NamedRerouteUsage()
  n11 = WorldPosition()
  NamedRerouteDeclaration(None=Multiply(A=VertexColor().R, B=RotateAboutAxis(NormalizedRotationAxis=@n8, RotationAngle=@n9, PivotPoint=@n10, Position=@n11.XYZ))) {name=Step1_WPO, nodeColor=<rgba 1 0 0.3984 1>}
  n12 = Transform(None=Constant3Vector() {constant=<rgba 0 0 1 0>}) {transformSourceType=TRANSFORMSOURCE_Local}
  n13 = MaterialFunctionCall(Float3=NamedRerouteUsage()) {materialFunction=&BreakOutFloat3Components}
  NamedRerouteDeclaration(None=Multiply(A=VertexColor().G, B=Multiply(A=@n12, B=@n13.G))) {name=Step2_WPO, nodeColor=<rgba 0 0.1562 1 1>}
  n14 = Transform(None=Constant3Vector() {constant=<rgba 0 0 1 0>}) {transformSourceType=TRANSFORMSOURCE_Local}
  n15 = Multiply(A=MaterialFunctionCall(Float3=NamedRerouteUsage()) {materialFunction=&BreakOutFloat3Components}.B) {constB=-1}
  NamedRerouteDeclaration(None=Multiply(A=VertexColor().B, B=Multiply(A=@n14, B=@n15))) {name=Step3_WPO, nodeColor=<rgba 0 0.1562 1 1>}
  n16 = ScalarParameter() {bUseCustomPrimitiveData=true, defaultValue=-1, parameterName=StartTime}
  n17 = ScalarParameter() {defaultValue=1, desc=The animation timelines can only go from 0-1 so use this to make the length longer or shorter., parameterName=TimeMult}
  n18 = Multiply(A=Subtract(A=Time() {period=0.0001}, B=@n16), B=@n17)
  n19 = CurveAtlasRowParameter(CurveTime=LinearInterpolate(
    A=If(A=@n16, "A > B"=@n18, "A == B"=@n16, "A < B"=@n16)
    B=ScalarParameter() {bUseCustomPrimitiveData=true, defaultValue=-1, parameterName=StartTime}
    Alpha=ScalarParameter() {bUseCustomPrimitiveData=true, parameterName=DEBUG_UseStartTimeAsTestValue, primitiveDataIndex=1}
   )) {
   atlas=&/Atlas_AugmentContainer
   curve=&/Curve_Anim_Open
   parameterName=AnimOpenCurve
  }
  NamedRerouteDeclaration(None=@n19) {name=CurveValue, nodeColor=<rgba 0.1016 1 0 1>}
  ScalarParameter() {defaultValue=50, parameterName=Step2_MoveAmount}
 ]
}
