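# Sound cue for completing an interaction with a legendary "Aug" container, an environmental item in MidMatchAssignedGameplay.
# A single mechanical wave feeds a Modulator that randomizes pitch, with output at reduced volume under a melee weapon sound class.
# Inline attenuation overrides the referenced TEMP gadget-activate asset, giving logarithmic falloff with distance-based low-pass filtering.
# The cue's debug flag is left on.
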SoundCue /MidMatchAssignedGameplay/Items/Environmental/Sounds/Aug_Legendary_Interact_Complete_Container_Cue
root=Modulator_0
nodes [
 {id=Modulator_0, children [WavePlayer_0], properties {pitchMax=1.1}}
 {
  id=WavePlayer_0
  properties {soundWaveAssetPtr=&/Aug_Legendary_Interact_Complete_Container_Mech}
 }
]
properties {
 attenuationSettings=/MidMatchAssignedGameplay/Items/Environmental/Sounds/TEMP/Fort_GadgetActivate_Att
 bDebug=true
 bOverrideAttenuation=true
 maxDistance=8500
 soundClassObject=&Fort_WeaponMelee_PSM
 volumeMultiplier=0.6
 attenuationOverrides {
  bAttenuateWithLPF=true
  distanceAlgorithm=Logarithmic
  falloffDistance=7000
  focusAzimuth=10
  lPFFrequencyAtMax=8000
  nonFocusAzimuth=30
  nonFocusVolumeAttenuation=0
  reverbDistanceMax=1024
  reverbDistanceMin=0
  reverbWetLevelMax=0
  reverbWetLevelMin=1
  attenuationShapeExtents {x=1500}
 }
}
