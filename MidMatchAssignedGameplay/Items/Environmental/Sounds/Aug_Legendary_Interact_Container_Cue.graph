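# Sound cue for interacting with a legendary "Aug" container in MidMatchAssignedGameplay, judging by its name and path.
# A single wave player plays the Team variant of the wave directly, with boosted volume, raised priority and virtualization disabled.
# Its own attenuation override sets distance falloff and occlusion parameters.
# Its concurrency override caps simultaneous voices at two, stopping the quietest with a short release.
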SoundCue /MidMatchAssignedGameplay/Items/Environmental/Sounds/Aug_Legendary_Interact_Container_Cue
root=WavePlayer_0
nodes [{id=WavePlayer_0, properties {soundWaveAssetPtr=&/Aug_Legendary_Interact_Container_Team}}]
properties {
 bOverrideAttenuation=true
 bOverrideConcurrency=true
 maxDistance=1500
 priority=85
 soundClassObject=/Game/Sounds/FortSoundClassesAndModes/Fort_LTM_BR_InGame_Phase_End_AMB_3D
 virtualizationMode=Disabled
 volumeMultiplier=1.35
 attenuationOverrides {
  falloffDistance=1000
  occlusionInterpolationTime=0.4
  occlusionTraceChannel=ECC_GameTraceChannel1
  occlusionVolumeAttenuation=0.7
  attenuationShapeExtents {x=500}
 }
 concurrencyOverrides {maxCount=2, resolutionRule=StopQuietest, voiceStealReleaseTime=0.2}
}
