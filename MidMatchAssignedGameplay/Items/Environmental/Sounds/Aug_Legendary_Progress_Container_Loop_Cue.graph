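# Looping sound cue, named for the progress loop of an Aug Legendary container among MidMatchAssignedGameplay environmental items, in the LTM gadgets sound class.
# A single looping wave player feeds the output directly, playing a SearchDestroy plant power-up wave from a TEMP folder.
# Overrides apply logarithmic distance falloff with low-pass filtering and stereo spread, and limit playback to one instance per owner, stopping the oldest.
# The cue's debug flag is left on.
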SoundCue /MidMatchAssignedGameplay/Items/Environmental/Sounds/Aug_Legendary_Progress_Container_Loop_Cue
root=WavePlayer_1
nodes [{
  id=WavePlayer_1
  properties {bLooping=true, soundWaveAssetPtr=/MidMatchAssignedGameplay/Items/Environmental/Sounds/TEMP/SearchDestroy_PlantPowerUp_Loop_01}
 }]
properties {
 bDebug=true
 bOverrideAttenuation=true
 bOverrideConcurrency=true
 maxDistance=6300
 soundClassObject=/Game/Sounds/FortSoundClassesAndModes/Fort_Gadgets_LTM
 volumeMultiplier=0.9
 attenuationOverrides {
  bAttenuateWithLPF=true
  distanceAlgorithm=Logarithmic
  falloffDistance=6000
  lPFFrequencyAtMax=8000
  reverbDistanceMax=1024
  reverbDistanceMin=0
  reverbWetLevelMax=0
  reverbWetLevelMin=1
  stereoSpread=800
  attenuationShapeExtents {x=300}
 }
 concurrencyOverrides {bLimitToOwner=true, maxCount=1, resolutionRule=StopOldest}
}
