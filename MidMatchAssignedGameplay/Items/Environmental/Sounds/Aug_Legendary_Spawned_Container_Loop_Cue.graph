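# Sound cue for a spawned legendary container loop in mid-match assigned gameplay, apparently an augment, built from temporary Scout Drone interact sounds.
# A mixer layers a modulated, randomly chosen start one-shot with a louder looping wave.
# An enveloper quickly fades in that loop, decays its volume over several seconds and shapes an early pitch bend.
# It uses the drone beam's attenuation, the gadgets sound class, capped concurrency and disabled virtualization.
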SoundCue /MidMatchAssignedGameplay/Items/Environmental/Sounds/Aug_Legendary_Spawned_Container_Loop_Cue
root=Mixer_0
nodes [
 {
  id=Enveloper_0
  children [WavePlayer_1]
  properties {
   pitchCurve {editorCurveData {keys [{time=0}, <at 0.3 1.1>, {time=1.25}]}}
   volumeCurve {editorCurveData {keys [
      <at 0 0>
      {arriveTangent=4.3335, interpMode=RCIM_Cubic, leaveTangent=4.3335, tangentMode=RCTM_User, time=0.2, value=1}
      {arriveTangent=-0.125, interpMode=RCIM_Cubic, leaveTangent=-0.125, time=0.2308, value=1}
      <at 5 0.4>
     ]}}
  }
 }
 {id=Mixer_0, children [Modulator_0,Enveloper_0], properties {inputVolume [1,1.5]}}
 {id=Modulator_0, children [Random_0]}
 {id=Random_0, children [WavePlayer_0,WavePlayer_2,WavePlayer_3], properties {bShouldExcludeFromBranchCulling=true, weights [1,1,1]}}
 {id=WavePlayer_0, properties {soundWaveAssetPtr=/MidMatchAssignedGameplay/Items/Environmental/Sounds/TEMP/Scout_Drone_Interact_Start_03}}
 {
  id=WavePlayer_1
  properties {bLooping=true, soundWaveAssetPtr=/MidMatchAssignedGameplay/Items/Environmental/Sounds/TEMP/Scout_Drone_Interact_Loop_01}
 }
 {id=WavePlayer_2, properties {soundWaveAssetPtr=/MidMatchAssignedGameplay/Items/Environmental/Sounds/TEMP/Scout_Drone_Interact_Start_01}}
 {id=WavePlayer_3, properties {soundWaveAssetPtr=/MidMatchAssignedGameplay/Items/Environmental/Sounds/TEMP/Scout_Drone_Interact_Start_02}}
]
properties {
 attenuationSettings=/MidMatchAssignedGameplay/Items/Environmental/Sounds/TEMP/Scout_Drone_Interact_Beam_Ambio_Att
 bOverrideConcurrency=true
 maxDistance=3000
 priority=0
 soundClassObject=&Fort_Gadgets
 virtualizationMode=Disabled
 volumeMultiplier=1
 concurrencyOverrides {maxCount=2}
}
