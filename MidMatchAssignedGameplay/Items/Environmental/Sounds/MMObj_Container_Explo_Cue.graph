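# Sound cue for a container explosion among the environmental items of MidMatchAssignedGameplay.
# A random node picks one of two equally weighted explosion waves and feeds a modulator for slight pitch variation.
# It plays through the medium explosives sound class with overridden attenuation that fades with distance but is not spatialized.
# It also references a Creative CTF delivery concurrency set.
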SoundCue /MidMatchAssignedGameplay/Items/Environmental/Sounds/MMObj_Container_Explo_Cue
root=Modulator_0
nodes [
 {id=Modulator_0, children [Random_0], properties {pitchMin=0.99}}
 {id=Random_0, children [WavePlayer_0,WavePlayer_1], properties {weights [1,1]}}
 {id=WavePlayer_0, properties {soundWaveAssetPtr=&/MmObj_Eplo_01}}
 {id=WavePlayer_1, properties {soundWaveAssetPtr=&/MmObj_Eplo_02}}
]
properties {
 bOverrideAttenuation=true
 maxDistance=9000
 priority=0.9
 soundClassObject=/Game/Sounds/FortSoundClassesAndModes/Fort_Explosives_Medium
 volumeMultiplier=0.25
 attenuationOverrides {bSpatialize=false, falloffDistance=6000, attenuationShapeExtents {x=3000}}
 concurrencyOverrides {resolutionRule=PreventNew}
 concurrencySet [/Game/Sounds/Creative/Concurrency/Creative_CTF_Delivery_Conc]
}
