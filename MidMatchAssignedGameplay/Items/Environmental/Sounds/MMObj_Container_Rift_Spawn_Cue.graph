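# Sound cue for the build-up of a rift container spawn among mid-match assigned gameplay items.
# A mixer layers a close branch and a slightly quieter distant branch, each with its own rift-encounter attenuation asset.
# Each branch randomly picks an equally weighted build-up variant and is excluded from branch culling.
# The cue plays in the gadgets sound class at reduced volume, with a concurrency override capping simultaneous instances.
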SoundCue /MidMatchAssignedGameplay/Items/Environmental/Sounds/MMObj_Container_Rift_Spawn_Cue
root=Mixer_0
nodes [
 {id=Attenuation_0, children [Random_0], properties {attenuationSettings=/RiftEncounters/Sound/RiftEncounter_Close_Att}}
 {id=Attenuation_1, children [Random_1], properties {attenuationSettings=/RiftEncounters/Sound/RiftEncounter_Distant_Att}}
 {id=Mixer_0, children [Attenuation_0, Attenuation_1], properties {inputVolume [1,0.7]}}
 {
  id=Random_0
  children [WavePlayer_1,WavePlayer_2,WavePlayer_3,WavePlayer_4]
  properties {bShouldExcludeFromBranchCulling=true, weights [1,1,1,1]}
 }
 {id=Random_1, children [WavePlayer_5,WavePlayer_6,WavePlayer_7], properties {bShouldExcludeFromBranchCulling=true, weights [1,1,1]}}
 {id=WavePlayer_1, properties {soundWaveAssetPtr=&/MMObj_Rift_Container_BuildUp_Close01}}
 {id=WavePlayer_2, properties {soundWaveAssetPtr=&/MMObj_Rift_Container_BuildUp_Close02}}
 {id=WavePlayer_3, properties {soundWaveAssetPtr=&/MMObj_Rift_Container_BuildUp_Close03}}
 {id=WavePlayer_4, properties {soundWaveAssetPtr=&/MMObj_Rift_Container_BuildUp_Close04}}
 {id=WavePlayer_5, properties {soundWaveAssetPtr=&/MMObj_Rift_Container_BuildUp_Distant01}}
 {id=WavePlayer_6, properties {soundWaveAssetPtr=&/MMObj_Rift_Container_BuildUp_Distant02}}
 {id=WavePlayer_7, properties {soundWaveAssetPtr=&/MMObj_Rift_Container_BuildUp_Distant03}}
]
properties {
 bOverrideConcurrency=true
 maxDistance=25000
 soundClassObject=&Fort_Gadgets
 volumeMultiplier=0.5
 concurrencyOverrides {maxCount=3}
}
