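# Sound cue for a mid-match assigned gameplay container spawning, routed to the gadgets sound class with a low concurrency cap.
# The close layer mixes one of four random spawn variants with an interact-complete container sound, passes through a modulator, and uses close attenuation.
# The distant layer plays a single spawn wave through distant attenuation at half mix volume, with a very large max distance on the cue.
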SoundCue /MidMatchAssignedGameplay/Items/Environmental/Sounds/MMObj_Container_Spawn_Cue
root=Mixer_0
nodes [
 {
  id=Attenuation_0
  children [Modulator_0]
  properties {attenuationSettings=&/MMObj_Close_Att}
 }
 {
  id=Attenuation_1
  children [WavePlayer_4]
  properties {attenuationSettings=&/MMObj_Distant_Att}
 }
 {id=Mixer_0, children [Attenuation_0, Attenuation_1], properties {inputVolume [1,0.5]}}
 {id=Mixer_1, children [Random_0,WavePlayer_2], properties {inputVolume [1,1]}}
 {id=Modulator_0, children [Mixer_1]}
 {
  id=Random_0
  children [WavePlayer_0,WavePlayer_8,WavePlayer_9,WavePlayer_1]
  properties {bShouldExcludeFromBranchCulling=true, weights [1,1,1,1]}
 }
 {id=WavePlayer_0, properties {soundWaveAssetPtr=&/MMObj_Container_Spawn_Close01}}
 {id=WavePlayer_1, properties {soundWaveAssetPtr=&/MMObj_Container_Spawn_Close04}}
 {
  id=WavePlayer_2
  properties {soundWaveAssetPtr=&/Aug_Legendary_Interact_Complete_Container}
 }
 {
  id=WavePlayer_4
  properties {soundWaveAssetPtr=&/MmObj_Regular_Container_Spawn_Distant_03}
 }
 {id=WavePlayer_8, properties {soundWaveAssetPtr=&/MMObj_Container_Spawn_Close02}}
 {id=WavePlayer_9, properties {soundWaveAssetPtr=&/MMObj_Container_Spawn_Close03}}
]
properties {
 bOverrideConcurrency=true
 maxDistance=160000
 soundClassObject=&Fort_Gadgets
 volumeMultiplier=0.7
 concurrencyOverrides {maxCount=3}
}
